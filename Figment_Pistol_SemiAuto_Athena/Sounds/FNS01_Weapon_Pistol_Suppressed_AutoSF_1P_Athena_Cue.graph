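# Sound cue for the first-person firing sound of a suppressed pistol in the Figment Athena semi-auto pistol content.
# A Random node picks between two equally weighted silenced-fire waves and feeds a Modulator at the root, which uses default settings.
# Playback is routed through shared first-person ranged-fire sound class, projectile attenuation and gunfire concurrency assets, with boosted volume.
# The cue's debug flag is left on.
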SoundCue /Figment_Pistol_SemiAuto_Athena/Sounds/FNS01_Weapon_Pistol_Suppressed_AutoSF_1P_Athena_Cue
root=Modulator_0
nodes [
 {id=Modulator_0, children [Random_0]}
 {id=Random_0, children [WavePlayer_5,WavePlayer_6], properties {weights [1,1]}}
 {id=WavePlayer_5, properties {soundWaveAssetPtr=/Game/Sounds/Fort_Weapons/pistol/Suppressed/Pistol_Silenced_Fire_01}}
 {id=WavePlayer_6, properties {soundWaveAssetPtr=/Game/Sounds/Fort_Weapons/pistol/Suppressed/Pistol_Silenced_Fire_02}}
]
properties {
 attenuationSettings=&Fort_Weapon_Projectile_FirstPerson_Athena_Att
 bDebug=true
 maxDistance=2097152
 soundClassObject=&Fort_WeaponRanged_Fire_1P_PSM
 volumeMultiplier=2
 attenuationOverrides {manualReverbSendLevel=-1.0165, reverbSendMethod=Manual}
 concurrencyOverrides {bLimitToOwner=true, maxCount=3, resolutionRule=StopOldest, volumeScale=0.9}
 concurrencySet [&Weapon_Gunfire_FirstPerson_Conc]
}
